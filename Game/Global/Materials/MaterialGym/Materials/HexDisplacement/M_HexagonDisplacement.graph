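# M_HexagonDisplacement is a Default Lit material in the MaterialGym folder, named for a hexagon displacement effect.
# It uses Material Attributes and holds a single node, a call to the MF_HexagonDisplacement material function.
# The effect's logic lives in that function, leaving this material as a thin wrapper around it.
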Material /Game/Global/Materials/MaterialGym/Materials/HexDisplacement/M_HexagonDisplacement
graph {
 properties {
  bUseMaterialAttributes=true
  pixelDepthOffsetMode=PDOM_Legacy
  previewMesh=/Engine/Transient.MaterialEditorMeshComponent_3
  usedShadingModels=Default Lit
 }
 nodes [MaterialFunctionCall!(In) {materialFunction=/Game/Global/Materials/MaterialGym/Materials/HexDisplacement/MaterialFunctions/MF_HexagonDisplacement}]
}
